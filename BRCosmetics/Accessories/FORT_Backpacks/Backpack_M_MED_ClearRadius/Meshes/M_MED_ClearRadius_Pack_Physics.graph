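# Physics asset for the ClearRadius backpack cosmetic, giving collision and joints to its dangling keychain.
# Capsule bodies cover the pack, the Key bone and three dyn_keychain bones.
# Constraints link Key and the keychain bones in a series, with a wide swing at the Key joint, tighter swing further along and little twist throughout.
# The large pack capsule has no constraint to the chain.
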PhysicsAsset /BRCosmetics/Accessories/FORT_Backpacks/Backpack_M_MED_ClearRadius/Meshes/M_MED_ClearRadius_Pack_Physics
bodies {
 Key [{shapeType=Capsule, length=25.6208, radius=1.856, center=<4.1506 0.5272 -2.9066>, rotation=<rot 0 -91.0711 -59.547>}]
 dyn_keychain_01 [{shapeType=Capsule, length=1.25, radius=2.5, center=<7.3623 0 0>, rotation=<rot 90 0 0>}]
 dyn_keychain_02 [{shapeType=Capsule, length=1, radius=2.9, center=<2 0 0>, rotation=<rot 90 0 0>}]
 dyn_keychain_03 [{shapeType=Capsule, length=1.75, radius=2.4, center=<2 0 0>, rotation=<rot 90 0 0>}]
 pack [{shapeType=Capsule, length=30, radius=42, shapeName=pack_capsule, center=<8.1731 42 0>, rotation=<rot 90 180 179.9995>}]
}
constraints [
 {bone1Name=dyn_keychain_01, bone2Name=Key, swing1LimitDegrees=170, swing2LimitDegrees=60, twistLimitDegrees=5}
 {bone1Name=dyn_keychain_02, bone2Name=dyn_keychain_01, swing1LimitDegrees=60, swing2LimitDegrees=40, twistLimitDegrees=5}
 {bone1Name=dyn_keychain_03, bone2Name=dyn_keychain_02, swing1LimitDegrees=60, swing2LimitDegrees=40, twistLimitDegrees=5}
]
