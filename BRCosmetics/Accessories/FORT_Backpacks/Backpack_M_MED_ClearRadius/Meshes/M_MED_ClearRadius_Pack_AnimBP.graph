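# Anim Blueprint for the ClearRadius backpack cosmetic, adding secondary motion to the pack and its dangling keychain.
# A cached Scalability pose selects RigidBody, AnimDynamics or Trail controller simulation from the enabled-physics flags.
# Characters tagged RestrictCape get the pack aimed at the body mesh's pelvis with corrective offsets; others get fixed pack and keychain offsets.
# Update logic derives movement flags, state-blended gravity, trail fake velocity and sprint-dependent inertia scales.
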AnimBlueprint /BRCosmetics/Accessories/FORT_Backpacks/Backpack_M_MED_ClearRadius/Meshes/M_MED_ClearRadius_Pack_AnimBP
target_skeleton=&/M_MED_ClearRadius_Pack_Skeleton
graphs [
 AnimGraph {nodes [
   n1 = Look At - Bone: pack!(
    ComponentPose=Local To Component(LocalPose=Use cached pose 'Scalability'())
    bAlphaBoolEnabled:bool=True
    LookAtLocation:Vector=Get PelvisLocation() -> (PelvisLocation:Vector)
   ) -> (Pose) {
    alphaInputType=Bool
    alphaBoolBlend {blendOption=Linear}
    boneToModify {boneName=pack}
    lookAt_Axis {axis=<1 0>}
    lookUp_Axis {axis {x=0, z=1}}
    lookAtTarget=~
   }
   n2 = "Transform (Modify) Bone - Bone: pack"!(ComponentPose=@n1, bAlphaBoolEnabled:bool, Translation="0,-0.6,0", Rotation="0,0,30", Scale) -> (Pose) {
    alphaInputType=Bool
    rotationMode=BMM_Additive
    translationMode=BMM_Additive
    alphaBoolBlend {blendOption=Linear}
    boneToModify {boneName=pack}
    rotation {roll=30}
    translation {y=-0.6}
   }
   n3 = "Transform (Modify) Bone - Bone: pack"(ComponentPose=@n2, Translation="0.5,-0.8,0", Rotation="0,0,20") {
    alphaInputType=Bool
    rotationMode=BMM_Additive
    translationMode=BMM_Additive
    alphaBoolBlend {blendOption=Linear}
    boneToModify {boneName=pack}
    rotation {roll=20}
    translation=<0.5 -0.8>
   }
   n4 = "Transform (Modify) Bone - Bone: pack"(ComponentPose=@n3, Translation="2,-2,0", Rotation="10,0,0") {
    alphaInputType=Bool
    rotationMode=BMM_Additive
    translationMode=BMM_Additive
    alphaBoolBlend {blendOption=Linear}
    boneToModify {boneName=pack}
    rotation {pitch=10}
    translation=<2 -2>
   }
   n5 = "Transform (Modify) Bone - Bone: pack"(ComponentPose=@n4, bAlphaBoolEnabled=false, Translation="0,-0.6,0", Rotation="0,0,10") {
    alphaInputType=Bool
    rotationMode=BMM_Additive
    translationMode=BMM_Additive
    alphaBoolBlend {blendOption=Linear}
    boneToModify {boneName=pack}
    rotation {roll=10}
    translation {y=-0.6}
   }
   n6 = "Transform (Modify) Bone - Bone: pack"(ComponentPose=Local To Component(LocalPose=Use cached pose 'Scalability'()), Translation="0,0,0", Rotation="-10,-20,22") {
    alphaInputType=Bool
    rotationMode=BMM_Additive
    scaleMode=BMM_Additive
    scaleSpace=BCS_ParentBoneSpace
    translationMode=BMM_Additive
    alphaBoolBlend {blendInTime=0.2, blendOption=Cubic, blendOutTime=0.2}
    boneToModify {boneName=pack}
    rotation=<rot -10 -20 22>
   }
   n7 = "Transform (Modify) Bone - Bone: dyn_keychain_01"!(ComponentPose=@n6, bAlphaBoolEnabled:bool=True, Translation="0,0,-1", Rotation="0,0,30", Scale="1,1,1") -> (Pose) {
    alphaInputType=Bool
    rotationMode=BMM_Additive
    scaleMode=BMM_Additive
    scaleSpace=BCS_ParentBoneSpace
    translationMode=BMM_Additive
    alphaBoolBlend {blendInTime=0.1, blendOption=Cubic, blendOutTime=0.1}
    boneToModify {boneName=dyn_keychain_01}
    rotation {roll=30}
    translation {z=-1}
   }
   Output Pose(Result=Blend Poses by bool(
     BlendPose_0=Component To Local(ComponentPose=@n5)
     BlendPose_1=Component To Local(ComponentPose=@n7)
     BlendTime_0=0.2
     BlendTime_1=0.2
     bActiveValue=Get BuiltInCape_Active() -> (BuiltInCape_Active:bool)
    ) {blendType=Linear}) {layerGroup=None}
   n8 = Trail controller - Bone: dyn_keychain_03!(
    ComponentPose=Local To Component(LocalPose=Use cached pose 'BasePose'())
    Alpha=0.4
    FakeVelocity:Vector=Get TrailFakeVelocity() -> (TrailFakeVelocity:Vector)
   ) -> (Pose) {
    bActorSpaceFakeVel=true
    bShowTrailLocation=true
    bUsePlanarLimit=true
    chainLength=3
    alphaBoolBlend {blendInTime=0.2, blendOption=Cubic, blendOutTime=0.2}
    baseJoint {boneName=root}
    fakeVelocity {z=500}
    planarLimits [
     {drivingBone {boneName=Key}, planeTransform {location=<0 3 0>, rotation=<rot 0 0 -75>, scale=<1 1 1>}}
     {drivingBone {boneName=Key}, planeTransform {location=<-10 0 0>, rotation=<rot -30 -30 0>, scale=<1 1 1>}}
    ]
    rotationLimits [{limitMax=<0 0 0>, limitMin=<0 0 0>}, {limitMax=<0 0 0>, limitMin=<0 0 0>}, {limitMax=<180 180 180>}]
    rotationOffsets [<0 0 0>, <0 0 0>, <0 0 0>]
    trailBone {boneName=dyn_keychain_03}
    trailRelaxationSpeed {editorCurveData {keys [<at 0 10 RCIM_Linear>,<at 0.3335 6.5439 RCIM_Linear>,<at 1 5 RCIM_Linear>]}}
   }
   n9 = Trail controller - Bone: dyn_keychain_03!(ComponentPose=@n8, bAlphaBoolEnabled:bool=Get IsFallnigAfterJump() -> (IsFallnigAfterJump:bool), FakeVelocity=Get TrailFakeVelocity()) -> (Pose) {
    alphaInputType=Bool
    bActorSpaceFakeVel=true
    bLimitRotation=true
    bShowTrailLocation=true
    bUsePlanarLimit=true
    chainLength=3
    relaxationSpeedScale=4
    alphaBoolBlend {blendInTime=0.4, blendOption=Cubic, blendOutTime=0.4}
    baseJoint {boneName=root}
    fakeVelocity {z=500}
    planarLimits [
     {drivingBone {boneName=Key}, planeTransform {location=<0 3 0>, rotation=<rot 0 0 -74.9999>, scale=<1 1 1>}}
     {drivingBone {boneName=Key}, planeTransform {location=<-10 0 0>, rotation=<rot -30 -30 0>, scale=<1 1 1>}}
    ]
    rotationLimits [
     {limitMax=<70 -3.6 180>, limitMin=<0 -180 -14.4002>}
     {limitMax=<120 -3.6 180>, limitMin=<0 -180 -14.4002>}
     {limitMax=<130 -3.6 180>, limitMin=<0 -180 -14.4002>}
    ]
    rotationOffsets [<0 0 0>, <0 0 0>, <0 0 0>]
    trailBone {boneName=dyn_keychain_03}
    trailRelaxationSpeed {editorCurveData {keys [<at 0 10 RCIM_Linear>,<at 0.3335 6.5439 RCIM_Linear>,<at 1 5 RCIM_Linear>]}}
   }
   n10 = Trail controller - Bone: dyn_keychain_03!(ComponentPose=Local To Component(LocalPose=Use cached pose 'BasePose'()), Alpha:float=0.2, FakeVelocity="0,0,0") -> (Pose) {
    bActorSpaceFakeVel=true
    bShowTrailLocation=true
    bUsePlanarLimit=true
    chainLength=3
    relaxationSpeedScale=7
    alphaBoolBlend {blendInTime=0.4, blendOption=Cubic, blendOutTime=0.4}
    baseJoint {boneName=root}
    fakeVelocity {z=500}
    planarLimits [
     {drivingBone {boneName=Key}, planeTransform {location=<0 3 0>, rotation=<rot 0 0 -74.9999>, scale=<1 1 1>}}
     {drivingBone {boneName=Key}, planeTransform {location=<-10 0 0>, rotation=<rot -30 -30 0>, scale=<1 1 1>}}
    ]
    rotationLimits [
     {limitMax=<110 -3.6 180>, limitMin=<0 -180 -14.4002>}
     {limitMax=<120 -3.6 180>, limitMin=<0 -180 -14.4002>}
     {limitMax=<130 -3.6 180>, limitMin=<0 -180 -14.4002>}
    ]
    rotationOffsets [<0 0 0>, <0 0 0>, <0 0 0>]
    trailBone {boneName=dyn_keychain_03}
    trailRelaxationSpeed {editorCurveData {keys [<at 0 10 RCIM_Linear>,<at 0.3335 6.5439 RCIM_Linear>,<at 1 5 RCIM_Linear>]}}
   }
   Save cached pose 'Trail'(Pose=Blend Poses by bool(
     BlendPose_0=Component To Local(ComponentPose=@n10)
     BlendPose_1=Component To Local(ComponentPose=@n9)
     BlendTime_0=0.4
     BlendTime_1=0.4
     bActiveValue=Get isBackardJumping() -> (isBackardJumping:bool)
    ) {blendType=Linear}) {cacheName=Trail}
   Save cached pose 'BasePose'(Pose=Input Pose()) {cacheName=BasePose}
   n11 = RigidBody!(
    ComponentPose=Local To Component(LocalPose=Use cached pose 'BasePose'())
    Alpha
    OverrideWorldGravity:Vector=Get Gravity Override()
    ExternalForce
    ComponentLinearAccScale:Vector=Get Component Linear Acc Scale() -> (Component Linear Acc Scale:Vector)
    ComponentLinearVelScale:Vector=Get Component Linear Vel Scale() -> (Component Linear Vel Scale:Vector)
   ) -> (Pose) {
    bClampLinearTranslationLimitToRefPose=true
    bOverrideWorldGravity=true
    overridePhysicsAsset=&/M_MED_ClearRadius_Pack_Physics
    alphaBoolBlend {blendOption=Linear}
    baseBoneRef {boneName=root}
    componentAppliedLinearAccClamp=<3000 3000 3000>
    componentLinearAccScale=<0.4 0.3 0.3>
    componentLinearVelScale=<0.4 0.3 0.3>
   }
   Save cached pose 'RBAN'(Pose=Component To Local(ComponentPose=@n11)) {cacheName=RBAN}
   n12 = Get Component Linear Vel Scale AnimDynamics() -> (Component Linear Vel Scale AnimDynamics:Vector)
   n13 = Get Gravity Override()
   n14 = Get Component Linear Acc Scale AnimDynamics() -> (Component Linear Acc Scale AnimDynamics:Vector)
   n15 = Anim Dynamics - Bone: dyn_keychain_01!(
    ComponentPose=Local To Component(LocalPose=Use cached pose 'BasePose'())
    Alpha=0.9
    GravityOverride:Vector=@n13
    ComponentLinearAccScale:Vector=@n14
    ComponentLinearVelScale:Vector=@n12
    ExternalForce="0,0,0"
   ) -> (Pose) {
    angularDampingOverride=0.99
    angularSpringConstant=600
    bAngularSpring=true
    bOverrideAngularDamping=true
    bUseGravityOverride=true
    alphaBoolBlend {blendOption=Linear}
    boundBone {boneName=dyn_keychain_01}
    componentLinearAccScale=<0.6 0.5 0.5>
    componentLinearVelScale=<0.6 0.5 0.5>
    physicsBodyDefinitions [{
      boundBone {boneName=dyn_keychain_01}
      boxExtents=<20 20 20>
      constraintSetup {angularTargetAxis=AxisY, angularLimitsMax=<0 30 2>, angularLimitsMin=<0 -30 -100>, angularTarget=<0 0 0>}
      localJointOffset=<10 0 0>
     }]
    relativeSpaceBone {boneName=root}
    retargetingSettings {source {scale=<0.5 0.3 0.1>}}
   }
   n16 = Anim Dynamics - Bone: dyn_keychain_02!(
    ComponentPose=@n15
    Alpha=1
    GravityOverride:Vector=@n13
    ComponentLinearAccScale:Vector=@n14
    ComponentLinearVelScale:Vector=@n12
    ExternalForce="0,0,0"
   ) -> (Pose) {
    angularDampingOverride=0.99
    angularSpringConstant=300
    bAngularSpring=true
    bOverrideAngularDamping=true
    bUseGravityOverride=true
    alphaBoolBlend {blendOption=Linear}
    boundBone {boneName=dyn_keychain_02}
    componentLinearAccScale=<0.6 0.5 0.5>
    componentLinearVelScale=<0.6 0.5 0.5>
    physicsBodyDefinitions [{
      boundBone {boneName=dyn_keychain_02}
      boxExtents=<20 20 20>
      constraintSetup {angularTargetAxis=AxisY, angularLimitsMax=<10 30 2>, angularLimitsMin=<-10 -50 -80>, angularTarget=<0 0 0>}
      localJointOffset=<10 0 0>
     }]
    relativeSpaceBone {boneName=root}
   }
   n17 = Anim Dynamics - Bone: dyn_keychain_03!(
    ComponentPose=@n16
    Alpha=1
    GravityOverride:Vector=@n13
    ComponentLinearAccScale:Vector=@n14
    ComponentLinearVelScale:Vector=@n12
    ExternalForce="0,0,0"
   ) -> (Pose) {
    angularDampingOverride=0.99
    angularSpringConstant=300
    bAngularSpring=true
    bOverrideAngularDamping=true
    bUseGravityOverride=true
    alphaBoolBlend {blendOption=Linear}
    boundBone {boneName=dyn_keychain_03}
    componentLinearAccScale=<0.6 0.5 0.5>
    componentLinearVelScale=<0.6 0.5 0.5>
    physicsBodyDefinitions [{
      boundBone {boneName=dyn_keychain_03}
      boxExtents=<20 20 20>
      constraintSetup {angularTargetAxis=AxisY, angularLimitsMax=<0 30 2>, angularLimitsMin=<0 -50 -80>, angularTarget=<0 0 0>}
      localJointOffset=<10 0 0>
     }]
    relativeSpaceBone {boneName=root}
   }
   n18 = "Transform (Modify) Bone - Bone: dyn_keychain_01"(
    ComponentPose=@n17
    bAlphaBoolEnabled=Get IsFallnigAfterJump()
    Translation="-0.379,1.447,-1.527"
    Rotation="-0.000,0.000,40.000"
    Scale="1,1,1"
   ) {
    alphaInputType=Bool
    translationMode=BMM_Additive
    alphaBoolBlend {blendInTime=0.2, blendOption=Cubic, blendOutTime=0.3}
    boneToModify {boneName=dyn_keychain_01}
    rotation {roll=40}
    translation=<-0.379 1.447 -1.527>
   }
   n19 = "Transform (Modify) Bone - Bone: dyn_keychain_01"(ComponentPose=@n18, bAlphaBoolEnabled=True, Translation="0.000,1.489,-0.000", Rotation="-0.000,0.000,40.000", Scale="1,1,1") {
    alphaInputType=Bool
    translationMode=BMM_Additive
    alphaBoolBlend {blendInTime=0.2, blendOption=Cubic, blendOutTime=0.2}
    boneToModify {boneName=dyn_keychain_01}
    rotation {roll=40}
    translation=<0 1.4893 0>
   }
   n20 = "Transform (Modify) Bone - Bone: dyn_keychain_01"(ComponentPose=@n19, bAlphaBoolEnabled=Get bIsSkyDiving(), Translation="-0.000,1.341,-0.655", Rotation="0,20,0", Scale="1,1,1") {
    alphaInputType=Bool
    rotationMode=BMM_Additive
    translationMode=BMM_Additive
    alphaBoolBlend {blendOption=Linear}
    boneToModify {boneName=dyn_keychain_01}
    rotation {yaw=20}
    translation {y=1.341, z=-0.655}
   }
   n21 = "Transform (Modify) Bone - Bone: dyn_keychain_01"(ComponentPose=@n20, bAlphaBoolEnabled=Get IsGlider() -> (IsGlider:bool), Translation="0,0,0", Rotation="0,0,30", Scale="1,1,1") {
    alphaInputType=Bool
    rotationMode=BMM_Additive
    translationMode=BMM_Additive
    alphaBoolBlend {blendInTime=0.2, blendOption=Cubic, blendOutTime=0.2}
    boneToModify {boneName=dyn_keychain_01}
    rotation {roll=30}
   }
   Save cached pose 'AnimDynamics'(Pose=Component To Local(ComponentPose=@n21)) {cacheName=AnimDynamics}
   n22 = Use cached pose 'Trail'()
   n23 = Use cached pose 'AnimDynamics'()
   n24 = Get bAnimDynamicsEnabled() -> (bAnimDynamicsEnabled:bool)
   Save cached pose 'Scalability'(Pose=Blend Poses by bool(
     BlendPose_0=Use cached pose 'RBAN'()
     BlendPose_1=Blend Poses by bool(BlendPose_0=@n23, BlendPose_1=@n22, bActiveValue=@n24) {blendType=Linear}
     BlendTime_0=0.2
     BlendTime_1=0.2
     bActiveValue=Get bRigidBodyEnabled()
    ) {blendType=Linear}) {cacheName=Scalability}
  ]}
 EventGraph {nodes [
   n1 = Event BlueprintUpdateAnimation()
   > Set DeltaTimeX(DeltaTimeX:float=@n1.DeltaTimeX) -> (Output_Get:float)
   n2 = Get bIsSkyDiving()
   > Set IsFallnigAfterJump(IsFallnigAfterJump:bool=AND Boolean(A=Get bIsFalling(), B=NOT Boolean(A=@n2))) -> (Output_Get:bool)
   > n3 = PlayerMovement(AnimInstance=Self-Reference(), XDotProductCoefficient=100, YDotProductCoefficient=100, ZDotProductCoefficient=100)
   n4 = "float < float"(A=BreakVector(InVec=MakeVector(X=@n3, Y=@n3.YMovement)).Y, B=-40)
   n5 = Get bIsJumping()
   n6 = Get IsFallnigAfterJump()
   > Set isBackardJumping(isBackardJumping:bool=OR Boolean(A=AND Boolean(A=@n4, B=@n5), B=AND Boolean(A=@n4, B=@n6))) -> (Output_Get:bool)
   n7 = Get bIsSkydiveFloating()
   > Set IsGlider(IsGlider:bool=AND Boolean(A=Get bIsSkyDiving(), B=NOT Boolean(A=@n7))) -> (Output_Get:bool)
   > n8 = Gravity Override State Function DEPRECATED(CharacterPartAnimInstance=Self-Reference(), GameTimeInSeconds=GetGameTimeInSeconds(), GravityOverrideParams=Get Gravity Override Params())
   n9 = AND Boolean(A=OR Boolean(A=Get bIsJumping(), B=Get bIsFalling()), B=NOT Boolean(A=Get bIsSkyDiving()))
   > Set AlphaJumping(AlphaJumping:float=FInterpTo(
     Current=Get AlphaJumping() -> (AlphaJumping:float)
     Target=SelectFloat(A=1, B=0, bPickA=@n9)
     DeltaTime=Get DeltaTimeX() -> (DeltaTimeX:float)
     InterpSpeed=10
    )) -> (Output_Get:float)
   > Set Gravity Override(Gravity Override="Lerp (Vector)"!(A:Vector=@n8, B:Vector=Get JumpingStateGravity() -> (JumpingStateGravity:Vector), Alpha:float=Get AlphaJumping()) -> (ReturnValue:Vector))
   > n10 = Branch(Condition=Get bIsSkyDiving())
   n11 = BreakVector(InVec=Get Gravity Override())
   > n12 = Set TrailFakeVelocity(TrailFakeVelocity:Vector=MakeVector(X=float * float(A=@n11, B=0.1) {is=title K2Node_PromotableOperator}, Y=float * float(A=@n11.Y, B=0.1), Z=float * float(A=@n11.Z, B=-0.25))) -> (Output_Get:Vector)
   n13 = Branch(execute=[@n12.then, @n21.then], Condition=Get bIsSliding())
   > n14 = Gravity Override State Function DEPRECATED(
    CharacterPartAnimInstance=Self-Reference()
    GameTimeInSeconds=GetGameTimeInSeconds()
    GravityOverrideParams=Get SlidingOverrideParams() -> (SlidingOverrideParams:Gravity Override Params Struct Structure)
   )
   > n15 = Set Gravity Override(Gravity Override=@n14)
   n16 = Get bodyMesh() -> (bodyMesh:Skeletal Mesh Component Object Reference)
   n17 = Is Valid(exec=[@n13.else, @n15.then], InputObject=@n16)
   n18 = SelectVector(A="0.9,0.3,0.3", B="0.4,0.3,0.3", bPickA=Get bIsTacticalSprinting())
   Set Component Linear Acc Scale(execute=[@n17.Is Not Valid, @n19.then], Component Linear Acc Scale:Vector=@n18) -> (Output_Get:Vector)
   > Set Component Linear Vel Scale(Component Linear Vel Scale:Vector=@n18) -> (Output_Get:Vector)
   n19 = Set PelvisLocation(
    execute=@n17.Is Valid
    PelvisLocation:Vector=GetSocketLocation!(self:Scene Component Object Reference=@n16, InSocketName:Name=pelvis) -> (ReturnValue:Vector)
   ) -> (Output_Get:Vector)
   n20 = BreakVector(InVec=Get Gravity Override())
   n21 = Set TrailFakeVelocity(
    execute=@n10.else
    TrailFakeVelocity=MakeVector(X=float * float(A=@n20, B=0.25), Y=float * float(A=@n20.Y, B=0.25), Z=float * float(A=@n20.Z, B=-0.1))
   )
   Event OnCharacterCustomizationLoaded()
   > n22 = Set BuiltInCape_Active(BuiltInCape_Active:bool=CharacterItemDefinitionHasAnyMetaTag(Tags=Make Array("[0]"=(TagName="Cosmetics.Character.RestrictCape")))) -> (Output_Get:bool)
   > Branch(Condition=@n22)
   > n23 = Cast To PlayerPawn_Athena(Object=TryGetPawnOwner())
   > Set BodyMesh(BodyMesh=GetSkeletalMeshForPartType(self=@n23))
   TryGetPawnOwner()
  ]}
]
variables [
 AlphaJumping {type=number}
 BodyMesh {type=SkeletalMeshComponent}
 BuiltInCape_Active {type=boolean}
 Component Linear Acc Scale {type=Vector, default=<0.4 0.3 0.3>}
 Component Linear Acc Scale AnimDynamics {type=Vector, default=<0.5 0.3 0.2>}
 Component Linear Vel Scale {type=Vector, default=<0.4 0.3 0.3>}
 Component Linear Vel Scale AnimDynamics {type=Vector, default=<0.5 0.3 0.2>}
 DeltaTimeX {type=number}
 Gravity Override {type=Vector}
 Gravity Override Params {
  type=GravityOverrideParamsStruct
  default {
   onGround_Moving {gravityOverride=<0 0 -800>}
   skydive_Parachuting {gravityOverride=<0 -1200 -980>}
   tactical_Sprint {gravityOverride=<0 -400 200>}
  }
 }
 isBackardJumping {type=boolean}
 IsFallnigAfterJump {type=boolean}
 IsGlider {type=boolean}
 JumpingStateGravity {type=Vector, default=<0 0 100>}
 PelvisLocation {type=Vector}
 SlidingOverrideParams {
  type=GravityOverrideParamsStruct
  default {
   emoteOrMelee {windAmplitude=<500 0 0>, windFrequency=<10000 0 0>}
   onGround_Moving {gravityOverride=<0 0 -400>, windAmplitude=<300 0 300>, windFrequency=<11 0 1>}
   skydive_Parachuting {gravityOverride=<0 -1200 -980>}
   tactical_Sprint {gravityOverride=<0 -400 200>}
  }
 }
 SlidingStateGravity {type=Vector}
 TrailFakeVelocity {type=Vector}
]
